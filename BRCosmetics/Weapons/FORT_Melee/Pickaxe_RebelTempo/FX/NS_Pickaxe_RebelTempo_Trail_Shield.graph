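# Niagara swing trail for the RebelTempo pickaxe (Shield variant), drawn as animation-trail ribbons in three emitters: TrailSharp, Trail_PBR and Trail_Toon.
# A variant switch picks either the Toon or PBR trail and TrailSharp's color.
# The swing section index sets lifetimes, shortened in front end, and selects the PBR and Toon alpha curves over emitter age.
# TrailSharp also tapers ribbon width and scales alpha by time of day.
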NiagaraSystem /BRCosmetics/Weapons/FORT_Melee/Pickaxe_RebelTempo/FX/NS_Pickaxe_RebelTempo_Trail_Shield
dependencies {
 usedDataInterfaces [&NiagaraDataInterfaceCurve, &NiagaraDataInterfaceSocketReader, &NiagaraDataInterfaceColorCurve]
 usedDynamicInputs [&MakeIntFromFloat, &ColorFromCurve]
 usedModules [
  &self:SystemSpawnScript.SystemScriptSource.SystemScriptGraph.NiagaraNodeAssignment_1.SetVariables_#1
  &SystemState
  &self:SystemSpawnScript.SystemScriptSource.SystemScriptGraph.NiagaraNodeAssignment_3.SetVariables_#2
  &self:SystemSpawnScript.SystemScriptSource.SystemScriptGraph.NiagaraNodeAssignment_4.SetVariables_#3
  &self:SystemSpawnScript.SystemScriptSource.SystemScriptGraph.NiagaraNodeAssignment_5.SetVariables_#4
  &self:SystemSpawnScript.SystemScriptSource.SystemScriptGraph.NiagaraNodeAssignment_0.SetVariables_#5
 ]
}
emitters {
 TrailSharp {
  emitterSpawnScript [&FN_RibbonIDOffsetUtility()]
  emitterUpdateScript [
   &EmitterState()
   &FN_SpawnAnimTrail()
  ]
  particleSpawnScript [
   &FN_InitializeAnimTrailBoneIndecies(First Bone or Socket Index=1)
   &FN_InitializeAnimTrail(
    AnimTrailWidth=3
    Color=&MergeHelper_ColorSwitch(0=<rgba 0.3026 0.4076 0.5 5>, 1=<rgba 0.3177 0.4883 1 5>, 2=?, 3=?, Index=$System.VariantSwitch, Num=::NewEnumerator0)
    Lifetime=&MergeHelper_FloatSwitch(
     0=0
     1=0.3
     2=0
     3=&FN_FrontEndFloatModifier(Gameplay=0.3, Front End=0.15)
     Index=$User.SwingSectionIndex
     Num=::NewEnumerator2
    )
   )
  ]
  particleUpdateScript [
   &SolveForcesAndVelocity(Clamp Velocity=false, Limit Acceleration=false)
   &FN_UpdateAnimTrail() {module=FN_UpdateAnimTrail001}
   &ScaleColor(
    Scale Alpha=&FloatFromCurve(
     FloatCurve=DataInterface() {propertyValues {Curve {keys [{interpMode=RCIM_Cubic, tangentMode=RCTM_User, time=0, value=1}, {interpMode=RCIM_Cubic, tangentMode=RCTM_User, time=1, value=0}]}}}
     Scale Curve=&FloatFromCurve(
      FloatCurve=DataInterface() {propertyValues {Curve {keys [{arriveTangent=-2.2175, interpMode=RCIM_Cubic, leaveTangent=-2.2175, tangentMode=RCTM_User, time=0.1, value=0.6}, <at 0.2 0 RCIM_Cubic>]}}}
      CurveIndex=$Emitter.Age
      Scale Curve=&FNTimeOfDayScalar(Night Scale=1.3, Morning Scale=1, Evening Scale=1)
     )
    )
    Scale RGB=&MakeVectorFromLinearColorRGB(Color=$System.ToDColor)
    ScaleA=true
    ScaleRGB=true
   )
   &ScaleRibbonWidth(Ribbon Width Scale=&FloatFromCurve(FloatCurve=DataInterface() {propertyValues {Curve {keys [<at 0 0.5>,<at 0.1 1>,<at 1 0.3>]}}}))
  ]
  rendererClasses [&NiagaraRibbonRendererProperties]
  renderers [{rendererClass=&NiagaraRibbonRendererProperties}]
 }
 Trail_PBR {
  emitterSpawnScript [&FN_RibbonIDOffsetUtility()]
  emitterUpdateScript [
   &EmitterState(Life Cycle Mode=::NewEnumerator1, Loop Behavior=::NewEnumerator1, Loop Duration=0.225, UseLoopDelay=false)
   &FN_SpawnAnimTrail(Number of Anim Trails:NiagaraInt32=&MergeHelper_IntSwitch(0=0, 1=1, 2=?, 3=?, 4=?, 5=?, Index=$System.VariantSwitch, Num=::NewEnumerator0))
  ]
  particleSpawnScript [
   &FN_InitializeAnimTrailBoneIndecies()
   &FN_InitializeAnimTrail(Lifetime=&MergeHelper_FloatSwitch(
     0=0
     1=0.3
     2=0
     3=&FN_FrontEndFloatModifier(Gameplay=0.3, Front End=0.15)
     Index=$User.SwingSectionIndex
     Num=::NewEnumerator2
    ))
  ]
  particleUpdateScript [
   &SolveForcesAndVelocity(Clamp Velocity=false, Limit Acceleration=false)
   &FN_UpdateAnimTrail() {module=FN_UpdateAnimTrail001}
   &ScaleColor(
    Scale Alpha=&MergeHelper_FloatSwitch(
     0=0
     1=&FloatFromCurve(
      FloatCurve=DataInterface() {propertyValues {Curve {keys [<at 0.2 1 RCIM_Cubic>, {arriveTangent=-8.6886, tangentMode=RCTM_User, time=0.3, value=0}]}}}
      CurveIndex=$Emitter.Age
      Scale Curve=1.2
     )
     2=0
     3=&FloatFromCurve(
      FloatCurve=DataInterface() {propertyValues {Curve {keys [<at 0.1657 1 RCIM_Cubic>, {arriveTangent=-26.5859, tangentMode=RCTM_User, time=0.2671, value=-0.0094}]}}}
      CurveIndex=$Emitter.Age
      Scale Curve=1.4
     )
     4=0
     Index=$User.SwingSectionIndex
     Num=::NewEnumerator3
    )
    Scale RGB=&MakeVectorFromLinearColorRGB(Color=$System.ToDColor)
    ScaleA=true
    ScaleRGB=true
   )
  ]
  rendererClasses [&NiagaraRibbonRendererProperties]
  renderers [{rendererClass=&NiagaraRibbonRendererProperties}]
 }
 Trail_Toon {
  emitterSpawnScript [&FN_RibbonIDOffsetUtility()]
  emitterUpdateScript [
   &EmitterState(Life Cycle Mode=::NewEnumerator1, Loop Behavior=::NewEnumerator1, Loop Duration=0.225, UseLoopDelay=false)
   &FN_SpawnAnimTrail(
    AnimTrailWidthMode=::FromSecond
    Number of Anim Trails:NiagaraInt32=&MergeHelper_IntSwitch(0=1, 1=0, 2=?, 3=?, 4=?, 5=?, Index=$System.VariantSwitch, Num=::NewEnumerator0)
   )
  ]
  particleSpawnScript [
   &FN_InitializeAnimTrailBoneIndecies()
   &FN_InitializeAnimTrail(Lifetime=&MergeHelper_FloatSwitch(
     0=0
     1=0.2
     2=0
     3=&FN_FrontEndFloatModifier(Gameplay=0.2, Front End=0.1)
     Index=$User.SwingSectionIndex
     Num=::NewEnumerator2
    ))
  ]
  particleUpdateScript [
   &SolveForcesAndVelocity(Clamp Velocity=false, Limit Acceleration=false)
   &FN_UpdateAnimTrail() {module=FN_UpdateAnimTrail001}
   &ScaleColor(
    Scale Alpha=&MergeHelper_FloatSwitch(
     0=0
     1=&FloatFromCurve(
      FloatCurve=DataInterface() {propertyValues {Curve {keys [<at 0.1604 1 RCIM_Cubic>, {arriveTangent=-8.6886, tangentMode=RCTM_User, time=0.2, value=0}]}}}
      CurveIndex=$Emitter.Age
      Scale Curve=1.8
     )
     2=0
     3=&FloatFromCurve(
      FloatCurve=DataInterface() {propertyValues {Curve {keys [<at 0.1657 1 RCIM_Cubic>, {arriveTangent=-26.5859, tangentMode=RCTM_User, time=0.22, value=-0.0094}]}}}
      CurveIndex=$Emitter.Age
      Scale Curve=2.8
     )
     4=0
     Index=$User.SwingSectionIndex
     Num=::NewEnumerator3
    )
    Scale RGB=&MakeVectorFromLinearColorRGB(Color=$System.ToDColor)
    ScaleA=true
    ScaleRGB=true
   )
  ]
  rendererClasses [&NiagaraRibbonRendererProperties]
  renderers [{rendererClass=&NiagaraRibbonRendererProperties}]
 }
}
summary {
 systemName=NS_Pickaxe_RebelTempo_Trail_Shield
 emitters [
  {emitterName=TrailSharp, rendererClasses [&NiagaraRibbonRendererProperties]}
  {emitterName=Trail_Toon, rendererClasses [&NiagaraRibbonRendererProperties]}
  {emitterName=Trail_PBR, rendererClasses [&NiagaraRibbonRendererProperties]}
 ]
 userVariables [
  User.bIsFrontEnd {type=NiagaraBool}
  User.bIsFrontEndPreview {type=NiagaraBool}
  User.PhysicalSurface {type=NiagaraInt32}
  User.RibbonIDOffset {type=NiagaraInt32}
  User.SwingSectionIndex {type=NiagaraInt32}
  User.VariantSwitch {type=NiagaraFloat, defaultValue=0, description=""}
 ]
}
